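# Additive unlit skeletal-mesh material for the Adrenaline Rush gadget, drawing a green-to-yellow sweep over a character.
# A softened reflection vector from a normal map, converted to view space, provides matcap-like UVs into a healing gradient texture.
# The Sweep parameter replaces time in the panner to position the band, which is tinted between outer and center colors and scaled by SafetyOpacity.
# A vertex-normal offset term suggests an inflated shell.
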Material /Game/Gadgets/Assets/AdrenalineRush/Materials/M_AdrenalineRushSweep_01
graph {
 properties {
  bUsedWithClothing=true
  bUsedWithMeshDeformer=true
  bUsedWithMorphTargets=true
  bUsedWithSkeletalMesh=true
  blendMode=BLEND_Additive
  pixelDepthOffsetMode=PDOM_Legacy
  shadingModel=MSM_Unlit
  usedShadingModels=Unlit
 }
 nodes [
  Multiply(A=VertexNormalWS(), B=ScalarParameter() {defaultValue=1.1, parameterName=WPO Offset}) {constB=2}
  n1 = ScalarParameter() {defaultValue=0.85, parameterName=RelectionVectorSoftness}
  n2 = TextureSampleParameter2D() {
   parameterName=Normals
   samplerType=SAMPLERTYPE_Normal
   texture=/Game/Characters/Player/Male/Medium/Bodies/M_Med_Soldier_01/Textures/T_M_MED_soldier_01_N
  }
  n3 = Transform(None=@n2.RGB)
  n4 = ComponentMask(None=Transform(None=MaterialFunctionCall!(Softness=@n1, Normal=@n3) {materialFunction=/Engine/Functions/Engine_MaterialFunctions01/Reflections/ReflectionVectorSoft}.Result) {transformSourceType=TRANSFORMSOURCE_World, transformType=TRANSFORM_View}) {g=true, r=true}
  n5 = TextureSample!(
   UVs=Panner(Coordinate=Abs(None=@n4), Time=ScalarParameter() {defaultValue=-0.3, parameterName=Sweep}) {speedY=1.5}
   Tex
   Level
   Apply View MipBias
  ) {mipValueMode=TMVM_MipLevel, texture=/Game/Gadgets/Assets/AdrenalineRush/Textures/T_AdrenalineRush_HealingGradient_01}
  n6 = VectorParameter() {parameterName=Color Outer, defaultValue=<rgba 0.8243 1 0.0259 0>}
  n7 = VectorParameter() {parameterName=Color Center, defaultValue=<rgba 0.15 3 0.375 0>}
  Multiply(A=Multiply(A=@n5.RGB, B=LinearInterpolate(A=@n6.RGB, B=@n7.RGB, Alpha=@n5.RGB)), B=ScalarParameter() {parameterName=SafetyOpacity})
  Sine(None=Time()) {period=4}
 ]
}
