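# Translucent unlit sprite material for the Adrenaline Rush gadget's heal effect, used on Niagara and particle sprites with responsive AA.
# The heal texture's red channel tinted by particle color passes through an emissive adjustment function for the glow.
# A second chain, apparently opacity, multiplies texture RGB by particle alpha and a contrast-adjusted, saturated camera distance fade, driven by the NearCameraDistanceFade parameter.
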Material /Game/Gadgets/Assets/AdrenalineRush/Materials/M_AdrenalineRush_Heal_01
graph {
 properties {
  bEnableResponsiveAA=true
  bUsedWithNiagaraSprites=true
  bUsedWithParticleSprites=true
  blendMode=BLEND_Translucent
  pixelDepthOffsetMode=PDOM_Legacy
  shadingModel=MSM_Unlit
  usedShadingModels=Unlit
 }
 nodes [
  n1 = TextureSampleParameter2D() {parameterName=Texture, texture=/Game/Effects/Fort_Effects/Textures/T_Wall_Heal}
  n2 = ParticleColor()
  MaterialFunctionCall(Emissive=Multiply(A=@n1.R, B=@n2.RGB)) {materialFunction=&MF_FX_EmissiveAdjustments}
  n3 = MaterialFunctionCall!(Distance=ScalarParameter() {defaultValue=50, parameterName=NearCameraDistanceFade}) {materialFunction=/Game/Effects/Fort_Effects/Materials/_Functions/MF_CameraDistanceFade}
  n4 = Constant() {r=1}
  Multiply(
   A=Multiply(A=@n1.RGB, B=@n2.A)
   B=Saturate(None=MaterialFunctionCall(In=@n3.Result, Contrast=@n4) {materialFunction=&CheapContrast}.Result)
  )
 ]
}
