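# Unlit, translucent, two-sided mesh-particle material for the shockwave effect of the Adrenaline Rush gadget.
# Vertex red blends a dark color toward two light colors mixed by Fresnel, then particle color tints the result for emissive adjustment.
# Two panning noise channels are summed, stepped and contrast-adjusted, then added to vertex red.
# Particle alpha erodes and fades this mask, which vertex alpha multiplies.
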Material /Game/Gadgets/Assets/AdrenalineRush/Materials/M_AdrenalineRush_Shockwave_01
graph {
 properties {
  bEnableResponsiveAA=true
  bUsedWithMeshParticles=true
  blendMode=BLEND_Translucent
  pixelDepthOffsetMode=PDOM_Legacy
  previewMesh=/Game/Gadgets/Assets/AdrenalineRush/Meshes/St_AdrenalineRush_Shockwave_01
  shadingModel=MSM_Unlit
  twoSided=true
  usedShadingModels=Unlit
 }
 nodes [
  n1 = VectorParameter() {parameterName=Darks, defaultValue=<rgba 0.3673 1.5 0.2475 0>}
  n2 = LinearInterpolate(
   A=VectorParameter() {parameterName=Lights A, defaultValue=<rgba 0.12 4 0.04 0>}.RGB
   B=VectorParameter() {parameterName=Lights B, defaultValue=<rgba 2.6569 5 3.0398 0>}.RGB
   Alpha=MaterialFunctionCall(
    Brightness=ScalarParameter() {defaultValue=1, parameterName=Fresnel Brightness}
    Exponent=ScalarParameter() {defaultValue=3, parameterName=Fresnel Exponent}
   ) {materialFunction=&MF_Fresnel_FX}.Fresnel
  )
  n3 = VertexColor()
  n4 = ParticleColor()
  MaterialFunctionCall(Emissive=Multiply(A=LinearInterpolate(A=@n1.RGB, B=@n2, Alpha=@n3.R), B=@n4.RGB)) {materialFunction=&MF_FX_EmissiveAdjustments}
  n5 = TextureCoordinate() {vTiling=0.1052}
  n6 = TextureCoordinate() {uTiling=2, vTiling=0.18}
  n7 = Add(
   A=TextureSample(UVs=Panner(Coordinate=@n5) {bFractionalPart=true, speedX=0.16, speedY=-0.2685}) {texture=&T_NoisePack_01}.R
   B=TextureSample(UVs=Panner(Coordinate=@n6) {bFractionalPart=true, speedX=-0.48, speedY=-0.44}) {texture=&T_NoisePack_01}.G
  )
  n8 = ScalarParameter() {defaultValue=2, parameterName=Steps}
  n9 = Divide(A=Ceil(None=Multiply(A=@n7, B=@n8)), B=@n8)
  n10 = Add(A=@n3.R, B=Power(Base=Divide(A=@n9), Exp=ScalarParameter() {defaultValue=3, parameterName=Noise Contrast}))
  Multiply(A=Multiply(A=OneMinus(None=@n4.A), B=Subtract(A=@n10, B=@n4.A)), B=VertexColor().A)
  TextureSample(UVs=Panner(Coordinate=TextureCoordinate() {vTiling=0.2}) {speedX=0.12, speedY=-0.5}) {texture=/Game/Effects/Fort_Effects/Textures/Plasma/T_PlasmaNoise_A_01}
 ]
}
